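# Looping stereo MetaSound for the BeastGirl Storm Claws idle sparks loop, attenuated and limited to one instance per owner.
# An ADSR envelope fades in the wave player on play, while smoothed movement speed lowers its volume as speed rises.
# A delayed trigger later eases the mixer gain down to a quieter bed.
# The InCooldown, jump, step, speed direction and IdleLoop array inputs are unconnected.
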MetaSoundSource /BeastGirlStormClawsGameplay/Assets/Audio/SoundCues/Abilities/MSS_BeastGirlStorm_Idle
inputs [
 InCooldown {type=Bool}
 Movement.Speed {type=Float}
 Movement.SpeedAzimuth {type=Float}
 Movement.SpeedElevation {type=Float}
 Sparks Loop {type=WaveAsset}
 Trigger.OnJump {type=Float}
 Trigger.OnStep {type=Trigger}
 UE.Source.OnPlay {type=Trigger}
 Waves.IdleLoop {type=WaveAsset:Array}
]
interfaces [Fort.Movement:1.0, UE.OutputFormat.Stereo:1.0, UE.Source:1.1]
outputs [
 "UE.OutputFormat.Mono.Audio:0" {type=Audio}
 "UE.OutputFormat.Stereo.Audio:0" {type=Audio}
 "UE.OutputFormat.Stereo.Audio:1" {type=Audio}
]
pages [Default {
  input_defaults {
   InCooldown=null
   Movement.Speed=0
   Movement.SpeedAzimuth=0
   Movement.SpeedElevation=0
   Sparks Loop=/BeastGirlStormClawsGameplay/Assets/Audio/Wavs/Abilities/StormBeastGirl_Idle_Loop
   Trigger.OnJump=0
   Trigger.OnStep=null
   UE.Source.OnPlay=null
   Waves.IdleLoop [&StormBeastGirl_Idle_Loop]
  }
  nodes [
   ADSREnvelopeADSREnvelope_0 = ADSR Envelope.ADSR Envelope:Audio(Trigger Attack=^UE.Source.OnPlay) {also {defaults {Attack Time=1.5, Sustain Level=0.7}}}
   WavePlayer_0 = UE.Wave Player:Stereo(Wave Asset=^Sparks Loop, Play=@ADSREnvelopeADSREnvelope_0.On Attack Triggered) {also {defaults {Loop="<MetaSoundBuilderResult.SUCCEEDED: 0>"}}}
   TriggerDelay_0 = UE.Trigger Delay(In=@WavePlayer_0.On Play) {also {defaults {Delay Time=8}}}
   n1 = UE.Subtract:Float(AdditionalOperands=MapRange.MapRange:Float(In=&InterpSpeed(FloatInput=^Movement.Speed) {also {defaults {FallingInterpSpeed=0.5, RisingInterpSpeed=0.2}}}.FloatOutput) {also {defaults {In Range B=500, Out Range B=0.5}}}.Out Value) {also {defaults {PrimaryOperand=1}}}
   AudioMixerAudioMixerStereo2_0 = "AudioMixer.Audio Mixer (Stereo, 2)"(
    Gain 0=UE.InterpTo:Audio(Target=Value.Value:Float(Set=@TriggerDelay_0.Out) {also {defaults {Init Value=1, Target Value=0.2}}}.Output Value) {also {defaults {Interp Time=2}}}.Value
    In 0 L=UE.Multiply:Audio(
     PrimaryOperand=UE.Multiply:Audio by Float(AdditionalOperands=@n1.Out, PrimaryOperand=@WavePlayer_0.Out Left).Out
     AdditionalOperands=UE.Reroute(Value=@ADSREnvelopeADSREnvelope_0.Out Envelope).Value
    ).Out
    In 0 R=UE.Multiply:Audio(
     AdditionalOperands=@ADSREnvelopeADSREnvelope_0.Out Envelope
     PrimaryOperand=UE.Multiply:Audio by Float(AdditionalOperands=@n1.Out, PrimaryOperand=@WavePlayer_0.Out Right).Out
    ).Out
   )
  ]
  links [
   {dst="@UE.OutputFormat.Stereo.Audio:0", src=AudioMixerAudioMixerStereo2_0.Out L}
   {dst="@UE.OutputFormat.Stereo.Audio:1", src=AudioMixerAudioMixerStereo2_0.Out R}
  ]
 }]
properties {
 assetImportData=&self:AssetImportData
 attenuationSettings=/BeastGirlStormClawsGameplay/Assets/Audio/ATT_BeastGirlStorm_Idle
 bLooping=true
 bOverrideConcurrency=true
 duration=10000
 maxDistance=2750
 numChannels=2
 outputFormat=Stereo
 sampleRate=48000
 soundClassObject=&Fort_WeaponRanged
 volume=0.1
 concurrencyOverrides {bLimitToOwner=true, maxCount=1}
}
