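# Sound cue named for the charge phase of the Beast Girl Storm Claws AOE ability, with no root node or sound nodes, so it plays no audio as it stands.
# It is classed as a melee weapon sound and references the shared special AOE slam attenuation asset, alongside its own long-range attenuation override values.
# A concurrency override limits it to a single active instance.
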SoundCue /BeastGirlStormClawsGameplay/Assets/Audio/SoundCues/Abilities/SC_BeastGirlStormClaws_AOE_Charge
root=null
nodes []
properties {
 attenuationSettings=/Game/Sounds/Attenuation/Fort_Weapon_Melee_Attack_Special_AOE_Slam_Att
 bOverrideConcurrency=true
 maxDistance=9413.3125
 soundClassObject=&Fort_WeaponMelee
 attenuationOverrides {falloffDistance=10000, attenuationShapeExtents {x=10000}}
 concurrencyOverrides {maxCount=1}
}
