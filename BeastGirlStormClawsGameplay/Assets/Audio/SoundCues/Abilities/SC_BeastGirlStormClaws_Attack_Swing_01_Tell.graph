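# Sound cue named for the tell of Beast Girl Storm Claws' first swing attack, with an empty node graph and no root, so it plays no audio as shown.
# It is set to the melee weapon sound class and uses the shared close-range stereo melee attack attenuation.
# A concurrency override limits it to one instance at a time.
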SoundCue /BeastGirlStormClawsGameplay/Assets/Audio/SoundCues/Abilities/SC_BeastGirlStormClaws_Attack_Swing_01_Tell
root=null
nodes []
properties {
 attenuationSettings=/Game/Sounds/Attenuation/Fort_Weapon_Melee_Attack_Close_Range_Stereo_Att
 bOverrideConcurrency=true
 maxDistance=7927
 soundClassObject=&Fort_WeaponMelee
 volumeMultiplier=1
 attenuationOverrides {falloffDistance=10000, attenuationShapeExtents {x=10000}}
 concurrencyOverrides {maxCount=1}
}
