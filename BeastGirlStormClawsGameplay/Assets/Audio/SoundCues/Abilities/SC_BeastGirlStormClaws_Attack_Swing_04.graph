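# Sound cue for the fourth Beast Girl Storm Claws attack swing, an uppercut sound on the melee weapon sound class with close-range melee attenuation.
# A mixer layers a random claw swing wave, boosted, with a random electric wave, each picked from three variants.
# Branches on islocal and isenemy set loudness (local full, enemy slightly lower, non-enemy much quieter) beneath a root modulator that randomizes pitch.
# Two concurrency sets, global and local, limit overlapping attack sounds.
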SoundCue /BeastGirlStormClawsGameplay/Assets/Audio/SoundCues/Abilities/SC_BeastGirlStormClaws_Attack_Swing_04
root=Modulator_0
nodes [
 {id=Branch_0, children [Modulator_2,Branch_3,Modulator_2], properties {boolParameterName=islocal}}
 {id=Branch_3, children [Modulator_4,Modulator_3,Modulator_4], properties {boolParameterName=isenemy}}
 {id=Mixer_0, children [Random_0,Random_1], properties {inputVolume [1.6,1]}}
 {id=Modulator_0, children [Branch_0], properties {pitchMax=1.2, pitchMin=0.9, volumeMax=1}}
 {id=Modulator_2, children [Mixer_0], properties {pitchMax=1, pitchMin=1, volumeMax=1, volumeMin=1}}
 {id=Modulator_3, children [Mixer_0], properties {pitchMax=1, pitchMin=1, volumeMax=0.25, volumeMin=0.25}}
 {id=Modulator_4, children [Mixer_0], properties {pitchMax=1, pitchMin=1, volumeMax=0.8, volumeMin=0.8}}
 {id=Random_0, children [WavePlayer_7,WavePlayer_8,WavePlayer_9], properties {weights [1,1,1]}}
 {id=Random_1, children [WavePlayer_0,WavePlayer_1,WavePlayer_4], properties {weights [1,1,1]}}
 {
  id=WavePlayer_0
  properties {soundWaveAssetPtr=/BeastGirlStormClawsGameplay/Assets/Audio/Wavs/Abilities/StormBeastGirl_Claw_Attack_Swing_UpperCut_Electric_01}
 }
 {
  id=WavePlayer_1
  properties {soundWaveAssetPtr=/BeastGirlStormClawsGameplay/Assets/Audio/Wavs/Abilities/StormBeastGirl_Claw_Attack_Swing_UpperCut_Electric_02}
 }
 {
  id=WavePlayer_4
  properties {soundWaveAssetPtr=/BeastGirlStormClawsGameplay/Assets/Audio/Wavs/Abilities/StormBeastGirl_Claw_Attack_Swing_UpperCut_Electric_03}
 }
 {
  id=WavePlayer_7
  properties {soundWaveAssetPtr=/BeastGirlStormClawsGameplay/Assets/Audio/Wavs/Abilities/StormBeastGirl_Claw_Attack_Swing_UpperCut_01}
 }
 {
  id=WavePlayer_8
  properties {soundWaveAssetPtr=/BeastGirlStormClawsGameplay/Assets/Audio/Wavs/Abilities/StormBeastGirl_Claw_Attack_Swing_UpperCut_02}
 }
 {
  id=WavePlayer_9
  properties {soundWaveAssetPtr=/BeastGirlStormClawsGameplay/Assets/Audio/Wavs/Abilities/StormBeastGirl_Claw_Attack_Swing_UpperCut_03}
 }
]
properties {
 attenuationSettings=/Game/Sounds/Attenuation/Fort_Weapon_Melee_Attack_Close_Range_Stereo_Att
 maxDistance=10058.75
 pitchMultiplier=1.1
 soundClassObject=&Fort_WeaponMelee
 volumeMultiplier=0.4
 attenuationOverrides {falloffDistance=10000, attenuationShapeExtents {x=10000}}
 concurrencyOverrides {maxCount=1, resolutionRule=PreventNew}
 concurrencySet [
  /BeastGirlStormClawsGameplay/Assets/Audio/CONC_BeastGirlStorm_Attacks_Global_Stop_Far_Old
  /BeastGirlStormClawsGameplay/Assets/Audio/CONC_BeastGirlStorm_Attacks_Local_Stop_Oldest
 ]
}
