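# Sound cue for the whoosh of the BeastGirl Storm Claws de-transform ability.
# A single wave player feeds the output directly, with no randomization or mixing.
# It uses the melee weapon sound class and borrows the close-range spawn attenuation asset.
# Two shared attack concurrency sets, one global and one local, limit overlapping instances, with names indicating farthest or oldest sounds stop first.
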SoundCue /BeastGirlStormClawsGameplay/Assets/Audio/SoundCues/Abilities/SC_BeastGirlStormClaws_DeTransform_Whoosh
root=WavePlayer_0
nodes [{
  id=WavePlayer_0
  properties {soundWaveAssetPtr=/BeastGirlStormClawsGameplay/Assets/Audio/Wavs/Abilities/BeastGirlStorm_DeTransform_Whoosh_01}
 }]
properties {
 attenuationSettings=/BeastGirlStormClawsGameplay/Assets/Audio/ATT_BeastGirlStorm_Spawn_Close
 maxDistance=5000
 soundClassObject=&Fort_WeaponMelee
 volumeMultiplier=1
 attenuationOverrides {falloffDistance=10000, attenuationShapeExtents {x=10000}}
 concurrencyOverrides {maxCount=1, resolutionRule=PreventNew}
 concurrencySet [
  /BeastGirlStormClawsGameplay/Assets/Audio/CONC_BeastGirlStorm_Attacks_Global_Stop_Far_Old
  /BeastGirlStormClawsGameplay/Assets/Audio/CONC_BeastGirlStorm_Attacks_Local_Stop_Oldest
 ]
}
