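# Sound cue for the default impact sound of the BeastGirl Storm Claws pickaxe-style attack.
# A root Random node picks one of eight equally weighted impact wave variations and is excluded from branch culling.
# Playback uses the shared Fort pickaxe impact attenuation and Fort_Impacts sound class, with boosted volume.
# Two BeastGirlStorm attack concurrency groups limit overlapping hits, named for global stop-farthest-oldest and local stop-oldest behavior.
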SoundCue /BeastGirlStormClawsGameplay/Assets/Audio/SoundCues/Abilities/SC_BeastGirlStormClaws_Default_Impact
root=Random_1
nodes [
 {
  id=Random_1
  children [WavePlayer_0,WavePlayer_1,WavePlayer_2,WavePlayer_3,WavePlayer_4,WavePlayer_5,WavePlayer_6,WavePlayer_7]
  properties {bShouldExcludeFromBranchCulling=true, weights [1,1,1,1,1,1,1,1]}
 }
 {id=WavePlayer_0, properties {soundWaveAssetPtr=/BeastGirlStormClawsGameplay/Assets/Audio/Wavs/Impacts/BeastGirlStorm_Impact_01}}
 {id=WavePlayer_1, properties {soundWaveAssetPtr=/BeastGirlStormClawsGameplay/Assets/Audio/Wavs/Impacts/BeastGirlStorm_Impact_02}}
 {id=WavePlayer_2, properties {soundWaveAssetPtr=/BeastGirlStormClawsGameplay/Assets/Audio/Wavs/Impacts/BeastGirlStorm_Impact_03}}
 {id=WavePlayer_3, properties {soundWaveAssetPtr=/BeastGirlStormClawsGameplay/Assets/Audio/Wavs/Impacts/BeastGirlStorm_Impact_04}}
 {id=WavePlayer_4, properties {soundWaveAssetPtr=/BeastGirlStormClawsGameplay/Assets/Audio/Wavs/Impacts/BeastGirlStorm_Impact_05}}
 {id=WavePlayer_5, properties {soundWaveAssetPtr=/BeastGirlStormClawsGameplay/Assets/Audio/Wavs/Impacts/BeastGirlStorm_Impact_06}}
 {id=WavePlayer_6, properties {soundWaveAssetPtr=/BeastGirlStormClawsGameplay/Assets/Audio/Wavs/Impacts/BeastGirlStorm_Impact_07}}
 {id=WavePlayer_7, properties {soundWaveAssetPtr=/BeastGirlStormClawsGameplay/Assets/Audio/Wavs/Impacts/BeastGirlStorm_Impact_08}}
]
properties {
 attenuationSettings=/Game/Sounds/Attenuation/Fort_Pickaxe_Impact_Att
 maxDistance=4300
 soundClassObject=&Fort_Impacts
 volumeMultiplier=1.589
 attenuationOverrides {falloffDistance=10000, attenuationShapeExtents {x=10000}}
 concurrencyOverrides {maxCount=3}
 concurrencySet [
  /BeastGirlStormClawsGameplay/Assets/Audio/CONC_BeastGirlStorm_Attacks_Global_Stop_Far_Old
  /BeastGirlStormClawsGameplay/Assets/Audio/CONC_BeastGirlStorm_Attacks_Local_Stop_Oldest
 ]
}
