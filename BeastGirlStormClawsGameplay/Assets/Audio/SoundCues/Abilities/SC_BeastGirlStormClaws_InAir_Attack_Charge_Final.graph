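# Sound cue for Beast Girl Storm Claws' final in-air attack charge, playing a single build-up VO wave.
# Branches on islocal and isenemy set volume: full for local or unset, otherwise 0.8 for enemy or unset and 0.25 for non-enemy.
# All paths share that wave through fixed-pitch modulators.
# It uses close attenuation, the player attacks/abilities sound class, and concurrency sets limiting overlapping attack sounds.
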SoundCue /BeastGirlStormClawsGameplay/Assets/Audio/SoundCues/Abilities/SC_BeastGirlStormClaws_InAir_Attack_Charge_Final
root=Branch_0
nodes [
 {id=Branch_0, children [Modulator_2,Branch_3,Modulator_2], properties {boolParameterName=islocal}}
 {id=Branch_3, children [Modulator_4,Modulator_3,Modulator_4], properties {boolParameterName=isenemy}}
 {id=Modulator_2, children [WavePlayer_0], properties {pitchMax=1, pitchMin=1, volumeMax=1, volumeMin=1}}
 {id=Modulator_3, children [WavePlayer_0], properties {pitchMax=1, pitchMin=1, volumeMax=0.25, volumeMin=0.25}}
 {id=Modulator_4, children [WavePlayer_0], properties {pitchMax=1, pitchMin=1, volumeMax=0.8, volumeMin=0.8}}
 {
  id=WavePlayer_0
  properties {soundWaveAssetPtr=/BeastGirlStormClawsGameplay/Assets/Audio/Wavs/VO/BeastGirlStormClaws_InAir_Attack_BuildUp_Final_01}
 }
]
properties {
 attenuationSettings=/BeastGirlStormClawsGameplay/Assets/Audio/ATT_BeastGirlStormClaws_Close
 maxDistance=5400
 soundClassObject=/Game/Sounds/FortSoundClassesAndModes/Fort_06_Player_Attacks_Abilities
 volumeMultiplier=1.501
 attenuationOverrides {falloffDistance=10000, attenuationShapeExtents {x=10000}}
 concurrencyOverrides {maxCount=1, resolutionRule=StopOldest}
 concurrencySet [
  /BeastGirlStormClawsGameplay/Assets/Audio/CONC_BeastGirlStorm_Attacks_Global_Stop_Far_Old
  /BeastGirlStormClawsGameplay/Assets/Audio/CONC_BeastGirlStorm_Attacks_Local_Stop_Oldest
 ]
}
