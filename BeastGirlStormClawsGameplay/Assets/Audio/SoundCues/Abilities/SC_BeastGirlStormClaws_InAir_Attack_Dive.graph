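# Sound cue for the dive of the Beast Girl Storm Claws in-air attack.
# A mixer layers the dive-start wave with a slightly delayed dive wave, both boosted.
# Nested islocal and isenemy branches route that shared mix through fixed-volume modulators, playing non-local non-enemies much quieter than local players or enemies.
# The cue is pitched up, uses melee dash attenuation, and references global and local attack concurrency sets.
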SoundCue /BeastGirlStormClawsGameplay/Assets/Audio/SoundCues/Abilities/SC_BeastGirlStormClaws_InAir_Attack_Dive
root=Branch_2
nodes [
 {id=Branch_2, children [Modulator_2,Branch_3,Modulator_2], properties {boolParameterName=islocal}}
 {id=Branch_3, children [Modulator_4,Modulator_3,Modulator_4], properties {boolParameterName=isenemy}}
 {id=Delay_0, children [WavePlayer_0], properties {delayMax=0.1, delayMin=0.1}}
 {id=Mixer_0, children [Delay_0,WavePlayer_1], properties {inputVolume [2,2]}}
 {id=Modulator_2, children [Mixer_0], properties {pitchMax=1, pitchMin=1, volumeMax=0.7, volumeMin=0.7}}
 {id=Modulator_3, children [Mixer_0], properties {pitchMax=1, pitchMin=1, volumeMax=0.25, volumeMin=0.25}}
 {id=Modulator_4, children [Mixer_0], properties {pitchMax=1, pitchMin=1, volumeMax=0.8, volumeMin=0.8}}
 {
  id=WavePlayer_0
  properties {soundWaveAssetPtr=/BeastGirlStormClawsGameplay/Assets/Audio/Wavs/Abilities/BeastGirlStormClaws_InAir_Attack_Dive_01}
 }
 {
  id=WavePlayer_1
  properties {soundWaveAssetPtr=/BeastGirlStormClawsGameplay/Assets/Audio/Wavs/Abilities/BeastGirlStormClaws_InAir_Attack_Dive_Start_01}
 }
]
properties {
 attenuationSettings=/Game/Sounds/Attenuation/Fort_Weapon_Melee_Attack_Special_Dash_Att
 maxDistance=10058.75
 pitchMultiplier=1.3
 soundClassObject=/Game/Sounds/FortSoundClassesAndModes/Fort_Creatures_Large_Critical_PSM
 volumeMultiplier=1
 attenuationOverrides {falloffDistance=10000, attenuationShapeExtents {x=10000}}
 concurrencyOverrides {bLimitToOwner=true, maxCount=1, resolutionRule=PreventNew}
 concurrencySet [
  /BeastGirlStormClawsGameplay/Assets/Audio/CONC_BeastGirlStorm_Attacks_Global_Stop_Far_Old
  /BeastGirlStormClawsGameplay/Assets/Audio/CONC_BeastGirlStorm_Attacks_Local_Stop_Oldest
 ]
}
