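# Sound cue for the looping audio of the BeastGirlStormClaws in-air dive attack.
# A Looping node repeats a single Wave Player, so the dive sound plays until stopped.
# It routes to the player attacks and abilities sound class and borrows the BeastGirlStorm idle attenuation asset.
# Global and local attack concurrency sets limit overlapping instances, named for stop-farthest-then-oldest and stop-oldest rules.
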SoundCue /BeastGirlStormClawsGameplay/Assets/Audio/SoundCues/Abilities/SC_BeastGirlStormClaws_InAir_Attack_Dive_Loop
root=Looping_0
nodes [
 {id=Looping_0, children [WavePlayer_2]}
 {
  id=WavePlayer_2
  properties {soundWaveAssetPtr=/BeastGirlStormClawsGameplay/Assets/Audio/Wavs/Abilities/BeastGirlStormClaws_InAir_Attack_Dive_Loop_01}
 }
]
properties {
 attenuationSettings=/BeastGirlStormClawsGameplay/Assets/Audio/ATT_BeastGirlStorm_Idle
 maxDistance=2750
 soundClassObject=/Game/Sounds/FortSoundClassesAndModes/Fort_06_Player_Attacks_Abilities
 volumeMultiplier=0.893
 attenuationOverrides {falloffDistance=10000, attenuationShapeExtents {x=10000}}
 concurrencyOverrides {maxCount=1}
 concurrencySet [
  /BeastGirlStormClawsGameplay/Assets/Audio/CONC_BeastGirlStorm_Attacks_Global_Stop_Far_Old
  /BeastGirlStormClawsGameplay/Assets/Audio/CONC_BeastGirlStorm_Attacks_Local_Stop_Oldest
 ]
}
