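# Sound cue for the lightning of the BeastGirl Storm Claws in-air attack, with limited concurrency and sidechain bus sends.
# Branches on islocal, then isenemy, set volume: full for local, reduced for enemies, much quieter for non-enemies.
# All branches share a mix of close, medium and distant layers, each with its own attenuation and sound class.
# The close layer mixes random impacts with delayed, quality-switched attack variations.
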SoundCue /BeastGirlStormClawsGameplay/Assets/Audio/SoundCues/Abilities/SC_BeastGirlStormClaws_InAir_Attack_Lightning
root=Branch_0
nodes [
 {
  id=Attenuation_118
  children [Mixer_0]
  properties {attenuationSettings=/BeastGirlStormClawsGameplay/Assets/Audio/ATT_BeastGirlStormClaws_Close, attenuationOverrides {stereoSpread=29871.3984}}
 }
 {
  id=Attenuation_119
  children [Random_4]
  properties {attenuationSettings=/BeastGirlStormClawsGameplay/Assets/Audio/ATT_BeastGirlStormClaws_Med}
 }
 {
  id=Attenuation_3
  children [Random_5]
  properties {
   attenuationSettings=/BeastGirlStormClawsGameplay/Assets/Audio/ATT_BeastGirlStormClaws_Dist
   attenuationOverrides {bAttenuateWithLPF=true, falloffDistance=2000, lPFRadiusMax=5000, lPFRadiusMin=2000, attenuationShapeExtents {x=3000}}
  }
 }
 {id=Branch_0, children [Modulator_2,Branch_3,Modulator_2], properties {boolParameterName=islocal}}
 {id=Branch_3, children [Modulator_4,Modulator_3,Modulator_4], properties {boolParameterName=isenemy}}
 {id=Delay_0, children [QualityLevel_0], properties {delayMax=0.2, delayMin=0.2}}
 {id=Mixer_0, children [Delay_0,Random_1], properties {inputVolume [1,1]}}
 {id=Mixer_1, children [SoundClass_77, SoundClass_76, SoundClass_2], properties {inputVolume [1,1,1]}}
 {id=Modulator_2, children [Mixer_1], properties {pitchMax=1, pitchMin=1, volumeMax=1, volumeMin=1}}
 {id=Modulator_3, children [Mixer_1], properties {pitchMax=1, pitchMin=1, volumeMax=0.25, volumeMin=0.25}}
 {id=Modulator_4, children [Mixer_1], properties {pitchMax=1, pitchMin=1, volumeMax=0.8, volumeMin=0.8}}
 {id=QualityLevel_0, children [Random_0,Random_0,Random_2,Random_2]}
 {
  id=Random_0
  children [WavePlayer_3,WavePlayer_4,WavePlayer_5,WavePlayer_6,WavePlayer_7,WavePlayer_8,WavePlayer_9]
  properties {weights [1,1,1,1,1,1,1]}
 }
 {id=Random_1, children [WavePlayer_0,WavePlayer_1,WavePlayer_2], properties {weights [1,1,1]}}
 {id=Random_2, children [WavePlayer_18, WavePlayer_19, WavePlayer_20, WavePlayer_21], properties {weights [1,1,1,1]}}
 {id=Random_4, children [WavePlayer_14, WavePlayer_15, WavePlayer_16, WavePlayer_17], properties {weights [1,1,1,1]}}
 {id=Random_5, children [WavePlayer_10, WavePlayer_11, WavePlayer_12, WavePlayer_13], properties {weights [1,1,1,1]}}
 {
  id=SoundClass_2
  children [Attenuation_3]
  properties {soundClassOverride=&Fort_WeaponRanged_Fire_3P_PSM}
 }
 {
  id=SoundClass_76
  children [Attenuation_119]
  properties {soundClassOverride=&Fort_WeaponRanged_Fire_3P_PSM}
 }
 {
  id=SoundClass_77
  children [Attenuation_118]
  properties {soundClassOverride=/Game/Sounds/FortSoundClassesAndModes/Fort_06_Player_Attacks_Abilities}
 }
 {
  id=WavePlayer_0
  properties {soundWaveAssetPtr=/BeastGirlStormClawsGameplay/Assets/Audio/Wavs/Abilities/StormBeastGirl_Claw_Lightning_Impact_01}
 }
 {
  id=WavePlayer_1
  properties {soundWaveAssetPtr=/BeastGirlStormClawsGameplay/Assets/Audio/Wavs/Abilities/StormBeastGirl_Claw_Lightning_Impact_02}
 }
 {
  id=WavePlayer_10
  properties {soundWaveAssetPtr=/BeastGirlStormClawsGameplay/Assets/Audio/Wavs/Abilities/StormBeastGirl_Claw_Lightning_Impact_Distant_01}
 }
 {
  id=WavePlayer_11
  properties {soundWaveAssetPtr=/BeastGirlStormClawsGameplay/Assets/Audio/Wavs/Abilities/StormBeastGirl_Claw_Lightning_Impact_Distant_02}
 }
 {
  id=WavePlayer_12
  properties {soundWaveAssetPtr=/BeastGirlStormClawsGameplay/Assets/Audio/Wavs/Abilities/StormBeastGirl_Claw_Lightning_Impact_Distant_03}
 }
 {
  id=WavePlayer_13
  properties {soundWaveAssetPtr=/BeastGirlStormClawsGameplay/Assets/Audio/Wavs/Abilities/StormBeastGirl_Claw_Lightning_Impact_Distant_04}
 }
 {
  id=WavePlayer_14
  properties {soundWaveAssetPtr=/BeastGirlStormClawsGameplay/Assets/Audio/Wavs/Abilities/StormBeastGirl_Claw_Lightning_Impact_Medium_01}
 }
 {
  id=WavePlayer_15
  properties {soundWaveAssetPtr=/BeastGirlStormClawsGameplay/Assets/Audio/Wavs/Abilities/StormBeastGirl_Claw_Lightning_Impact_Medium_02}
 }
 {
  id=WavePlayer_16
  properties {soundWaveAssetPtr=/BeastGirlStormClawsGameplay/Assets/Audio/Wavs/Abilities/StormBeastGirl_Claw_Lightning_Impact_Medium_03}
 }
 {
  id=WavePlayer_17
  properties {soundWaveAssetPtr=/BeastGirlStormClawsGameplay/Assets/Audio/Wavs/Abilities/StormBeastGirl_Claw_Lightning_Impact_Medium_04}
 }
 {
  id=WavePlayer_18
  properties {soundWaveAssetPtr=/BeastGirlStormClawsGameplay/Assets/Audio/Wavs/Abilities/BeastGirlStormClaws_InAir_Attack_Lightning_04}
 }
 {
  id=WavePlayer_19
  properties {soundWaveAssetPtr=/BeastGirlStormClawsGameplay/Assets/Audio/Wavs/Abilities/BeastGirlStormClaws_InAir_Attack_Lightning_03}
 }
 {
  id=WavePlayer_2
  properties {soundWaveAssetPtr=/BeastGirlStormClawsGameplay/Assets/Audio/Wavs/Abilities/StormBeastGirl_Claw_Lightning_Impact_03}
 }
 {
  id=WavePlayer_20
  properties {soundWaveAssetPtr=/BeastGirlStormClawsGameplay/Assets/Audio/Wavs/Abilities/BeastGirlStormClaws_InAir_Attack_Lightning_02}
 }
 {
  id=WavePlayer_21
  properties {soundWaveAssetPtr=/BeastGirlStormClawsGameplay/Assets/Audio/Wavs/Abilities/BeastGirlStormClaws_InAir_Attack_Lightning_01}
 }
 {
  id=WavePlayer_3
  properties {soundWaveAssetPtr=/BeastGirlStormClawsGameplay/Assets/Audio/Wavs/Abilities/BeastGirlStormClaws_InAir_Attack_Lightning_07}
 }
 {
  id=WavePlayer_4
  properties {soundWaveAssetPtr=/BeastGirlStormClawsGameplay/Assets/Audio/Wavs/Abilities/BeastGirlStormClaws_InAir_Attack_Lightning_06}
 }
 {
  id=WavePlayer_5
  properties {soundWaveAssetPtr=/BeastGirlStormClawsGameplay/Assets/Audio/Wavs/Abilities/BeastGirlStormClaws_InAir_Attack_Lightning_05}
 }
 {
  id=WavePlayer_6
  properties {soundWaveAssetPtr=&BeastGirlStormClaws_InAir_Attack_Lightning_04}
 }
 {
  id=WavePlayer_7
  properties {soundWaveAssetPtr=&BeastGirlStormClaws_InAir_Attack_Lightning_03}
 }
 {
  id=WavePlayer_8
  properties {soundWaveAssetPtr=&BeastGirlStormClaws_InAir_Attack_Lightning_02}
 }
 {
  id=WavePlayer_9
  properties {soundWaveAssetPtr=&BeastGirlStormClaws_InAir_Attack_Lightning_01}
 }
]
properties {
 attenuationSettings=/BeastGirlStormClawsGameplay/Assets/Audio/ATT_BeastGirlStormClaws_Lightning
 maxDistance=30100
 soundClassObject=&Fort_06_Player_Attacks_Abilities
 volumeMultiplier=0.379
 attenuationOverrides {falloffDistance=10000, attenuationShapeExtents {x=10000}}
 busSends [
  {audioBus=&AB_Sidechain_FortWeapon_Vol, minSendLevel=0, sendLevel=0.1, soundSourceBus=null}
  {
   audioBus=/Game/Sounds/Modulation/EnvelopeFollowers/Volume/AB_SideChain_FortAmbient_Outdoors_Vol
   minSendLevel=0
   sendLevel=1
   soundSourceBus=null
  }
  {audioBus=&AB_Sidechain_FortAmbient_Vol, minSendLevel=0, sendLevel=1, soundSourceBus=null}
  {audioBus=&AB_Sidechain_FortExplosion_Vol, minSendLevel=0, sendLevel=0.4, soundSourceBus=null}
 ]
 concurrencyOverrides {maxCount=2, resolutionRule=StopOldest, retriggerTime=0.2, voiceStealReleaseTime=1}
 concurrencySet [
  /BeastGirlStormClawsGameplay/Assets/Audio/CONC_BeastGirlStorm_Lightning_Global
  /BeastGirlStormClawsGameplay/Assets/Audio/CONC_BeastGirlStorm_Lightning_Local
 ]
}
