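# Sound cue for the start of the BeastGirl StormClaws run ability, classed as melee weapon audio.
# A mixer layers a randomly chosen run-start wave, one of four equally weighted variants at reduced volume, with a fixed in-air dive attack start wave.
# The cue lowers overall pitch and volume, uses melee special-dash attenuation, and overrides concurrency to cap simultaneous instances.
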SoundCue /BeastGirlStormClawsGameplay/Assets/Audio/SoundCues/Abilities/SC_BeastGirlStormClaws_Run_Start
root=Mixer_0
nodes [
 {id=Mixer_0, children [Random_1,WavePlayer_1], properties {inputVolume [0.5,1]}}
 {
  id=Random_1
  children [WavePlayer_2,WavePlayer_3,WavePlayer_4,WavePlayer_5]
  properties {bShouldExcludeFromBranchCulling=true, weights [1,1,1,1]}
 }
 {
  id=WavePlayer_1
  properties {soundWaveAssetPtr=/BeastGirlStormClawsGameplay/Assets/Audio/Wavs/Abilities/BeastGirlStormClaws_InAir_Attack_Dive_Start_01}
 }
 {id=WavePlayer_2, properties {soundWaveAssetPtr=/BeastGirlStormClawsGameplay/Assets/Audio/Wavs/Abilities/StormBeastGirl_Run_Start_04}}
 {id=WavePlayer_3, properties {soundWaveAssetPtr=/BeastGirlStormClawsGameplay/Assets/Audio/Wavs/Abilities/StormBeastGirl_Run_Start_03}}
 {id=WavePlayer_4, properties {soundWaveAssetPtr=/BeastGirlStormClawsGameplay/Assets/Audio/Wavs/Abilities/StormBeastGirl_Run_Start_02}}
 {id=WavePlayer_5, properties {soundWaveAssetPtr=/BeastGirlStormClawsGameplay/Assets/Audio/Wavs/Abilities/StormBeastGirl_Run_Start_01}}
]
properties {
 attenuationSettings=/Game/Sounds/Attenuation/Fort_Weapon_Melee_Attack_Special_Dash_Att
 bOverrideConcurrency=true
 maxDistance=10058.75
 pitchMultiplier=0.8
 soundClassObject=&Fort_WeaponMelee
 volumeMultiplier=0.5
 attenuationOverrides {falloffDistance=10000, attenuationShapeExtents {x=10000}}
 concurrencyOverrides {maxCount=2}
}
