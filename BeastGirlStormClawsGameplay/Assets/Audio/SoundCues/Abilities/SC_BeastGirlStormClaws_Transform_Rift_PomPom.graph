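# Sound cue for the rift pom-pom transform ability sound of BeastGirl StormClaws, classed under player attacks and abilities.
# A mixer takes a rift pom-pom wave and a whoosh foley wave, slightly boosting the first and setting the foley input to zero volume, so only the pom-pom layer is audible.
# The cue lowers overall volume and pitch, references close-range melee attenuation, and overrides concurrency to cap simultaneous instances.
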SoundCue /BeastGirlStormClawsGameplay/Assets/Audio/SoundCues/Abilities/SC_BeastGirlStormClaws_Transform_Rift_PomPom
root=Mixer_0
nodes [
 {id=Mixer_0, children [WavePlayer_0,WavePlayer_1], properties {inputVolume [1.1,0]}}
 {
  id=WavePlayer_0
  properties {soundWaveAssetPtr=/BeastGirlStormClawsGameplay/Assets/Audio/Wavs/Abilities/BeastGirlStorm_Transform_Rift_PomPom_01}
 }
 {
  id=WavePlayer_1
  properties {soundWaveAssetPtr=/BeastGirlStormClawsGameplay/Assets/Audio/Wavs/Abilities/BeastGirlStorm_Transform_Whoosh_Foley_01}
 }
]
properties {
 attenuationSettings=/Game/Sounds/Attenuation/Fort_Weapon_Melee_Attack_Close_Range_Stereo_Att
 bOverrideConcurrency=true
 maxDistance=10058.75
 pitchMultiplier=0.9
 soundClassObject=/Game/Sounds/FortSoundClassesAndModes/Fort_06_Player_Attacks_Abilities
 volumeMultiplier=0.4
 attenuationOverrides {falloffDistance=10000, attenuationShapeExtents {x=10000}}
 concurrencyOverrides {maxCount=2}
}
