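# Sound cue for the lead-in audio of the BeastGirl StormClaws transform rift spawn, classed as a player ability sound.
# It plays a single lead-in wave directly from one Wave Player, with no randomization or modulation.
# Playback runs at reduced volume, spatialized through the BeastGirlStorm spawn attenuation asset.
# A concurrency override caps how many instances play at once.
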SoundCue /BeastGirlStormClawsGameplay/Assets/Audio/SoundCues/Abilities/SC_BeastGirlStormClaws_Transform_Rift_Spawn_LeadIn
root=WavePlayer_1
nodes [{
  id=WavePlayer_1
  properties {soundWaveAssetPtr=/BeastGirlStormClawsGameplay/Assets/Audio/Wavs/Abilities/BeastGirlStormClaws_Transform_LeadIn_01}
 }]
properties {
 attenuationSettings=/BeastGirlStormClawsGameplay/Assets/Audio/ATT_BeastGirlStorm_Spawn
 bOverrideConcurrency=true
 maxDistance=5000
 soundClassObject=/Game/Sounds/FortSoundClassesAndModes/Fort_06_Player_Attacks_Abilities
 volumeMultiplier=0.5
 attenuationOverrides {falloffDistance=10000, attenuationShapeExtents {x=10000}}
 concurrencyOverrides {maxCount=2}
}
