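# Sound cue for the whoosh of the BeastGirl StormClaws transform ability, grouped with melee weapon sounds through its sound class.
# A single wave player feeds the whoosh recording straight to the output, with no randomization or mixing.
# It uses close-range stereo melee attenuation and overrides concurrency to limit simultaneous plays.
# The cue's volume multiplier is zero, so it is silent as authored.
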SoundCue /BeastGirlStormClawsGameplay/Assets/Audio/SoundCues/Abilities/SC_BeastGirlStormClaws_Transform_Whoosh
root=WavePlayer_1
nodes [{id=WavePlayer_1, properties {soundWaveAssetPtr=/BeastGirlStormClawsGameplay/Assets/Audio/Wavs/Abilities/BeastGirlStorm_Transform_Whoosh_01}}]
properties {
 attenuationSettings=/Game/Sounds/Attenuation/Fort_Weapon_Melee_Attack_Close_Range_Stereo_Att
 bOverrideConcurrency=true
 maxDistance=10058.75
 soundClassObject=&Fort_WeaponMelee
 volumeMultiplier=0
 attenuationOverrides {falloffDistance=10000, attenuationShapeExtents {x=10000}}
 concurrencyOverrides {maxCount=2}
}
