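# Sound cue for the Storm Beast Girl's jump, routed to the creature foley sound class.
# A mixer layers one of two equally weighted jump foley waves at full volume with a quieter in-air jump start ability wave.
# The ability layer is pitched up by a fixed modulator.
# The cue doubles its output volume and references the character's footstep attenuation and concurrency assets.
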SoundCue /BeastGirlStormClawsGameplay/Assets/Audio/SoundCues/Abilities/SC_BeastGirlStorm_Jump
root=Mixer_0
nodes [
 {id=Mixer_0, children [Random_1,Random_0], properties {inputVolume [1,0.3]}}
 {id=Modulator_0, children [WavePlayer_3], properties {pitchMax=1.3, pitchMin=1.3, volumeMax=1, volumeMin=1}}
 {id=Random_0, children [Modulator_0], properties {weights [1]}}
 {id=Random_1, children [WavePlayer_0,WavePlayer_1], properties {weights [1,1]}}
 {id=WavePlayer_0, properties {soundWaveAssetPtr=/BeastGirlStormClawsGameplay/Assets/Audio/Wavs/Foley/StormBeastGirl_Jump_01}}
 {id=WavePlayer_1, properties {soundWaveAssetPtr=/BeastGirlStormClawsGameplay/Assets/Audio/Wavs/Foley/StormBeastGirl_Jump_02}}
 {
  id=WavePlayer_3
  properties {soundWaveAssetPtr=/BeastGirlStormClawsGameplay/Assets/Audio/Wavs/Abilities/BeastGirlStormClaws_InAir_Jump_Start_01}
 }
]
properties {
 attenuationSettings=/BeastGirlStormClawsGameplay/Assets/Audio/ATT_BeastGirlStorm_Footsteps
 maxDistance=3220.3438
 soundClassObject=/Game/Sounds/FortSoundClassesAndModes/Fort_Creatures_Foley
 volumeMultiplier=2
 attenuationOverrides {falloffDistance=10000, attenuationShapeExtents {x=10000}}
 concurrencyOverrides {maxCount=1, resolutionRule=PreventNew}
 concurrencySet [/BeastGirlStormClawsGameplay/Assets/Audio/CONC_BeastGirlStorm_FS]
}
