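# Sound cue for the Beast Girl Storm creature's landing foley, using footstep attenuation and concurrency assets.
# Branches on islocal and isenemy set loudness: full for local or unset, half for non-local enemies or unset enemy state, quarter for non-enemies.
# All paths share a quality-level node feeding randomized pitch and volume modulators over two land waves.
# The higher-quality random node is excluded from branch culling; both random nodes use the same wave pair.
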SoundCue /BeastGirlStormClawsGameplay/Assets/Audio/SoundCues/Abilities/SC_BeastGirlStorm_Land
root=Branch_0
nodes [
 {id=Branch_0, children [Modulator_2,Branch_3,Modulator_2], properties {boolParameterName=islocal}}
 {id=Branch_3, children [Modulator_4,Modulator_3,Modulator_4], properties {boolParameterName=isenemy}}
 {id=Modulator_0, children [Random_0], properties {pitchMax=1.1, volumeMax=1.2, volumeMin=0.7}}
 {id=Modulator_2, children [QualityLevel_0], properties {pitchMax=1, pitchMin=1, volumeMax=1, volumeMin=1}}
 {id=Modulator_3, children [QualityLevel_0], properties {pitchMax=1, pitchMin=1, volumeMax=0.25, volumeMin=0.25}}
 {id=Modulator_4, children [QualityLevel_0], properties {pitchMax=1, pitchMin=1, volumeMax=0.5, volumeMin=0.5}}
 {id=Modulator_5, children [Random_1], properties {pitchMax=1.1, volumeMax=1.2, volumeMin=0.7}}
 {id=QualityLevel_0, children [Modulator_5,Modulator_5,Modulator_0,Modulator_0]}
 {id=Random_0, children [WavePlayer_10, WavePlayer_11], properties {bShouldExcludeFromBranchCulling=true, weights [1,1]}}
 {id=Random_1, children [WavePlayer_8,WavePlayer_9], properties {weights [1,1]}}
 {id=WavePlayer_10, properties {soundWaveAssetPtr=/BeastGirlStormClawsGameplay/Assets/Audio/Wavs/Foley/StormBeastGirl_Land_02}}
 {id=WavePlayer_11, properties {soundWaveAssetPtr=/BeastGirlStormClawsGameplay/Assets/Audio/Wavs/Foley/StormBeastGirl_Land_01}}
 {id=WavePlayer_8, properties {soundWaveAssetPtr=&StormBeastGirl_Land_02}}
 {id=WavePlayer_9, properties {soundWaveAssetPtr=&StormBeastGirl_Land_01}}
]
properties {
 attenuationSettings=/BeastGirlStormClawsGameplay/Assets/Audio/ATT_BeastGirlStorm_Footsteps
 maxDistance=3220.3438
 soundClassObject=/Game/Sounds/FortSoundClassesAndModes/Fort_Creatures_Foley
 volumeMultiplier=1.7
 attenuationOverrides {falloffDistance=10000, attenuationShapeExtents {x=10000}}
 concurrencyOverrides {maxCount=1, resolutionRule=PreventNew}
 concurrencySet [/BeastGirlStormClawsGameplay/Assets/Audio/CONC_BeastGirlStorm_FS]
}
