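# Sound cue for the BeastGirl StormClaws low-health warning, combining an ability sound effect with a character voice line.
# A Mixer plays the effect wave immediately alongside the VO wave, which a fixed short Delay holds back so it lands just after the effect starts.
# The boosted cue uses Fortnite's low-health sound class and references the character's VO death concurrency set.
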SoundCue /BeastGirlStormClawsGameplay/Assets/Audio/SoundCues/Abilities/SC_BeastGirlStorm_LowHealth_Warning
root=Mixer_0
nodes [
 {id=Delay_0, children [WavePlayer_0], properties {delayMax=0.3, delayMin=0.3}}
 {id=Mixer_0, children [Delay_0,WavePlayer_2], properties {inputVolume [1,1]}}
 {id=WavePlayer_0, properties {soundWaveAssetPtr=/BeastGirlStormClawsGameplay/Assets/Audio/Wavs/VO/BeastGirlStorm_LowHealth_Start_VO}}
 {id=WavePlayer_2, properties {soundWaveAssetPtr=/BeastGirlStormClawsGameplay/Assets/Audio/Wavs/Abilities/BeastGirlStorm_LowHealth_Start}}
]
properties {
 maxDistance=2097152
 soundClassObject=/Game/Sounds/FortSoundClassesAndModes/Fort_LowHealth_PSM
 volumeMultiplier=1.8
 attenuationOverrides {falloffDistance=10000, attenuationShapeExtents {x=10000}}
 concurrencyOverrides {maxCount=1, resolutionRule=PreventNew}
 concurrencySet [/BeastGirlStormClawsGameplay/Assets/Audio/CONC_BeastGirlStorm_VO_Death]
}
