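# Sound cue for the Beast Girl Storm back-feet run footsteps, played as player foley with charge footstep attenuation.
# Branches on islocal and isenemy set loudness: full for local, half for enemies, quarter for other non-enemies, with unset parameters taking the louder path.
# All paths feed a quality-level switch choosing random pools of seven waves for the first two levels, four for the others.
# A footstep concurrency set limits overlap.
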SoundCue /BeastGirlStormClawsGameplay/Assets/Audio/SoundCues/Abilities/SC_BeastGirlStorm_Run_BackFeet
root=Branch_0
nodes [
 {id=Branch_0, children [Modulator_2,Branch_3,Modulator_2], properties {boolParameterName=islocal}}
 {id=Branch_3, children [Modulator_4,Modulator_3,Modulator_4], properties {boolParameterName=isenemy}}
 {id=Modulator_2, children [QualityLevel_0], properties {pitchMax=1, pitchMin=1, volumeMax=1, volumeMin=1}}
 {id=Modulator_3, children [QualityLevel_0], properties {pitchMax=1, pitchMin=1, volumeMax=0.25, volumeMin=0.25}}
 {id=Modulator_4, children [QualityLevel_0], properties {pitchMax=1, pitchMin=1, volumeMax=0.5, volumeMin=0.5}}
 {id=QualityLevel_0, children [Random_5,Random_5,Random_0,Random_0]}
 {
  id=Random_0
  children [WavePlayer_6, WavePlayer_7, WavePlayer_20, WavePlayer_21]
  properties {bShouldExcludeFromBranchCulling=true, weights [1,1,1,1]}
 }
 {
  id=Random_5
  children [WavePlayer_0,WavePlayer_1,WavePlayer_2,WavePlayer_3,WavePlayer_4,WavePlayer_5,WavePlayer_8]
  properties {bShouldExcludeFromBranchCulling=true, weights [1,1,1,1,1,1,1]}
 }
 {id=WavePlayer_0, properties {soundWaveAssetPtr=/BeastGirlStormClawsGameplay/Assets/Audio/Wavs/Foley/BeastGirlStorm_BackFeet_01}}
 {id=WavePlayer_1, properties {soundWaveAssetPtr=/BeastGirlStormClawsGameplay/Assets/Audio/Wavs/Foley/BeastGirlStorm_BackFeet_02}}
 {id=WavePlayer_2, properties {soundWaveAssetPtr=/BeastGirlStormClawsGameplay/Assets/Audio/Wavs/Foley/BeastGirlStorm_BackFeet_03}}
 {id=WavePlayer_20, properties {soundWaveAssetPtr=&BeastGirlStorm_BackFeet_03}}
 {id=WavePlayer_21, properties {soundWaveAssetPtr=/BeastGirlStormClawsGameplay/Assets/Audio/Wavs/Foley/BeastGirlStorm_BackFeet_04}}
 {id=WavePlayer_3, properties {soundWaveAssetPtr=&BeastGirlStorm_BackFeet_04}}
 {id=WavePlayer_4, properties {soundWaveAssetPtr=/BeastGirlStormClawsGameplay/Assets/Audio/Wavs/Foley/BeastGirlStorm_BackFeet_05}}
 {id=WavePlayer_5, properties {soundWaveAssetPtr=/BeastGirlStormClawsGameplay/Assets/Audio/Wavs/Foley/BeastGirlStorm_BackFeet_06}}
 {id=WavePlayer_6, properties {soundWaveAssetPtr=&BeastGirlStorm_BackFeet_01}}
 {id=WavePlayer_7, properties {soundWaveAssetPtr=&BeastGirlStorm_BackFeet_02}}
 {id=WavePlayer_8, properties {soundWaveAssetPtr=/BeastGirlStormClawsGameplay/Assets/Audio/Wavs/Foley/BeastGirlStorm_BackFeet_07}}
]
properties {
 attenuationSettings=/BeastGirlStormClawsGameplay/Assets/Audio/ATT_BeastGirlStorm_Footsteps_Charge
 maxDistance=5945.25
 soundClassObject=&Fort_Player_Foley
 volumeMultiplier=0.7
 attenuationOverrides {falloffDistance=10000, attenuationShapeExtents {x=10000}}
 concurrencyOverrides {maxCount=1, resolutionRule=PreventNew}
 concurrencySet [/BeastGirlStormClawsGameplay/Assets/Audio/CONC_BeastGirlStorm_FS]
}
