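# Sound cue for BeastGirlStorm's front-hand running foley, played with charge footstep attenuation and a footstep concurrency set.
# Branches on islocal and isenemy set loudness: full for local, half for enemies, quarter for other non-local players.
# All paths share a quality-level switch that picks from six random waves at the first two levels and four at the last two.
# Each pool gets randomized volume and pitch.
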SoundCue /BeastGirlStormClawsGameplay/Assets/Audio/SoundCues/Abilities/SC_BeastGirlStorm_Run_FrontHands
root=Branch_0
nodes [
 {id=Branch_0, children [Modulator_2,Branch_3,Modulator_2], properties {boolParameterName=islocal}}
 {id=Branch_3, children [Modulator_4,Modulator_3,Modulator_4], properties {boolParameterName=isenemy}}
 {id=Modulator_0, children [Random_0], properties {pitchMax=1.1, volumeMax=1.2, volumeMin=0.7}}
 {id=Modulator_2, children [QualityLevel_0], properties {pitchMax=1, pitchMin=1, volumeMax=1, volumeMin=1}}
 {id=Modulator_3, children [QualityLevel_0], properties {pitchMax=1, pitchMin=1, volumeMax=0.25, volumeMin=0.25}}
 {id=Modulator_4, children [QualityLevel_0], properties {pitchMax=1, pitchMin=1, volumeMax=0.5, volumeMin=0.5}}
 {id=Modulator_5, children [Random_5], properties {pitchMax=1.1, volumeMax=1.2, volumeMin=0.7}}
 {id=QualityLevel_0, children [Modulator_5,Modulator_5,Modulator_0,Modulator_0]}
 {
  id=Random_0
  children [WavePlayer_20, WavePlayer_21, WavePlayer_22, WavePlayer_23]
  properties {bShouldExcludeFromBranchCulling=true, weights [1,1,1,1]}
 }
 {
  id=Random_5
  children [WavePlayer_0,WavePlayer_1,WavePlayer_2,WavePlayer_3,WavePlayer_4,WavePlayer_5]
  properties {bShouldExcludeFromBranchCulling=true, weights [1,1,1,1,1,1]}
 }
 {id=WavePlayer_0, properties {soundWaveAssetPtr=/BeastGirlStormClawsGameplay/Assets/Audio/Wavs/Foley/BeastGirlStorm_FrontHands_01}}
 {id=WavePlayer_1, properties {soundWaveAssetPtr=/BeastGirlStormClawsGameplay/Assets/Audio/Wavs/Foley/BeastGirlStorm_FrontHands_02}}
 {id=WavePlayer_2, properties {soundWaveAssetPtr=/BeastGirlStormClawsGameplay/Assets/Audio/Wavs/Foley/BeastGirlStorm_FrontHands_03}}
 {id=WavePlayer_20, properties {soundWaveAssetPtr=&BeastGirlStorm_FrontHands_01}}
 {id=WavePlayer_21, properties {soundWaveAssetPtr=&BeastGirlStorm_FrontHands_02}}
 {id=WavePlayer_22, properties {soundWaveAssetPtr=&BeastGirlStorm_FrontHands_03}}
 {id=WavePlayer_23, properties {soundWaveAssetPtr=/BeastGirlStormClawsGameplay/Assets/Audio/Wavs/Foley/BeastGirlStorm_FrontHands_04}}
 {id=WavePlayer_3, properties {soundWaveAssetPtr=&BeastGirlStorm_FrontHands_04}}
 {id=WavePlayer_4, properties {soundWaveAssetPtr=/BeastGirlStormClawsGameplay/Assets/Audio/Wavs/Foley/BeastGirlStorm_FrontHands_05}}
 {id=WavePlayer_5, properties {soundWaveAssetPtr=/BeastGirlStormClawsGameplay/Assets/Audio/Wavs/Foley/BeastGirlStorm_FrontHands_06}}
]
properties {
 attenuationSettings=/BeastGirlStormClawsGameplay/Assets/Audio/ATT_BeastGirlStorm_Footsteps_Charge
 maxDistance=5945.25
 soundClassObject=&Fort_WeaponMelee
 volumeMultiplier=0.7
 attenuationOverrides {falloffDistance=10000, attenuationShapeExtents {x=10000}}
 concurrencyOverrides {maxCount=1, resolutionRule=PreventNew}
 concurrencySet [/BeastGirlStormClawsGameplay/Assets/Audio/CONC_BeastGirlStorm_FS]
}
