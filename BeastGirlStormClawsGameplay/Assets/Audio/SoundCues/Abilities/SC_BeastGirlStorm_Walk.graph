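# Footstep sound cue for the BeastGirlStorm character's walk, layering Walk and Step foley with its own footstep attenuation and concurrency.
# Branches on islocal and isenemy set loudness: full for local, reduced for enemies, quietest for other non-local players.
# All paths share a quality switch: the first two levels use eight-variant random pools with quiet steps, the last two four-variant pools at equal mix.
# Modulators randomize pitch and volume for variation.
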SoundCue /BeastGirlStormClawsGameplay/Assets/Audio/SoundCues/Abilities/SC_BeastGirlStorm_Walk
root=Branch_0
nodes [
 {id=Branch_0, children [Modulator_2,Branch_3,Modulator_2], properties {boolParameterName=islocal}}
 {id=Branch_3, children [Modulator_4,Modulator_3,Modulator_4], properties {boolParameterName=isenemy}}
 {id=Mixer_0, children [Random_5,Random_1], properties {inputVolume [1,0.1]}}
 {id=Mixer_1, children [Random_2,Modulator_0], properties {inputVolume [1,1]}}
 {id=Modulator_0, children [Random_0], properties {pitchMax=1.1, volumeMax=1.2, volumeMin=0.7}}
 {id=Modulator_2, children [QualityLevel_0], properties {pitchMax=1, pitchMin=1, volumeMax=1, volumeMin=1}}
 {id=Modulator_3, children [QualityLevel_0], properties {pitchMax=1, pitchMin=1, volumeMax=0.1, volumeMin=0.1}}
 {id=Modulator_4, children [QualityLevel_0], properties {pitchMax=1, pitchMin=1, volumeMax=0.3, volumeMin=0.3}}
 {id=Modulator_5, children [Mixer_0], properties {pitchMax=1.2, pitchMin=0.8, volumeMax=1.2, volumeMin=0.3}}
 {id=QualityLevel_0, children [Modulator_5,Modulator_5,Mixer_1,Mixer_1]}
 {
  id=Random_0
  children [WavePlayer_14, WavePlayer_15, WavePlayer_16, WavePlayer_17]
  properties {bShouldExcludeFromBranchCulling=true, weights [1,1,1,1]}
 }
 {
  id=Random_1
  children [WavePlayer_18, WavePlayer_19, WavePlayer_26, WavePlayer_27, WavePlayer_28, WavePlayer_29, WavePlayer_30, WavePlayer_31]
  properties {bShouldExcludeFromBranchCulling=true, weights [1,1,1,1,1,1,1,1]}
 }
 {id=Random_2, children [WavePlayer_2,WavePlayer_3,WavePlayer_4,WavePlayer_5], properties {weights [1,1,1,1]}}
 {
  id=Random_5
  children [WavePlayer_0, WavePlayer_1, WavePlayer_8, WavePlayer_9, WavePlayer_10, WavePlayer_11, WavePlayer_12, WavePlayer_13]
  properties {bShouldExcludeFromBranchCulling=true, weights [1,1,1,1,1,1,1,1]}
 }
 {id=WavePlayer_0, properties {soundWaveAssetPtr=/BeastGirlStormClawsGameplay/Assets/Audio/Wavs/Foley/BeastGirlStorm_Walk_01}}
 {id=WavePlayer_1, properties {soundWaveAssetPtr=/BeastGirlStormClawsGameplay/Assets/Audio/Wavs/Foley/BeastGirlStorm_Walk_02}}
 {id=WavePlayer_10, properties {soundWaveAssetPtr=/BeastGirlStormClawsGameplay/Assets/Audio/Wavs/Foley/BeastGirlStorm_Walk_05}}
 {id=WavePlayer_11, properties {soundWaveAssetPtr=/BeastGirlStormClawsGameplay/Assets/Audio/Wavs/Foley/BeastGirlStorm_Walk_06}}
 {id=WavePlayer_12, properties {soundWaveAssetPtr=/BeastGirlStormClawsGameplay/Assets/Audio/Wavs/Foley/BeastGirlStorm_Walk_07}}
 {id=WavePlayer_13, properties {soundWaveAssetPtr=/BeastGirlStormClawsGameplay/Assets/Audio/Wavs/Foley/BeastGirlStorm_Walk_08}}
 {id=WavePlayer_14, properties {soundWaveAssetPtr=&BeastGirlStorm_Walk_01}}
 {id=WavePlayer_15, properties {soundWaveAssetPtr=&BeastGirlStorm_Walk_02}}
 {id=WavePlayer_16, properties {soundWaveAssetPtr=/BeastGirlStormClawsGameplay/Assets/Audio/Wavs/Foley/BeastGirlStorm_Walk_03}}
 {id=WavePlayer_17, properties {soundWaveAssetPtr=/BeastGirlStormClawsGameplay/Assets/Audio/Wavs/Foley/BeastGirlStorm_Walk_04}}
 {id=WavePlayer_18, properties {soundWaveAssetPtr=/BeastGirlStormClawsGameplay/Assets/Audio/Wavs/Foley/StormBeastGirl_Step_08}}
 {id=WavePlayer_19, properties {soundWaveAssetPtr=/BeastGirlStormClawsGameplay/Assets/Audio/Wavs/Foley/StormBeastGirl_Step_07}}
 {id=WavePlayer_2, properties {soundWaveAssetPtr=/BeastGirlStormClawsGameplay/Assets/Audio/Wavs/Foley/StormBeastGirl_Step_04}}
 {id=WavePlayer_26, properties {soundWaveAssetPtr=/BeastGirlStormClawsGameplay/Assets/Audio/Wavs/Foley/StormBeastGirl_Step_06}}
 {id=WavePlayer_27, properties {soundWaveAssetPtr=/BeastGirlStormClawsGameplay/Assets/Audio/Wavs/Foley/StormBeastGirl_Step_05}}
 {id=WavePlayer_28, properties {soundWaveAssetPtr=&StormBeastGirl_Step_04}}
 {id=WavePlayer_29, properties {soundWaveAssetPtr=/BeastGirlStormClawsGameplay/Assets/Audio/Wavs/Foley/StormBeastGirl_Step_03}}
 {id=WavePlayer_3, properties {soundWaveAssetPtr=&StormBeastGirl_Step_03}}
 {id=WavePlayer_30, properties {soundWaveAssetPtr=/BeastGirlStormClawsGameplay/Assets/Audio/Wavs/Foley/StormBeastGirl_Step_02}}
 {id=WavePlayer_31, properties {soundWaveAssetPtr=/BeastGirlStormClawsGameplay/Assets/Audio/Wavs/Foley/StormBeastGirl_Step_01}}
 {id=WavePlayer_4, properties {soundWaveAssetPtr=&StormBeastGirl_Step_02}}
 {id=WavePlayer_5, properties {soundWaveAssetPtr=&StormBeastGirl_Step_01}}
 {id=WavePlayer_8, properties {soundWaveAssetPtr=&BeastGirlStorm_Walk_03}}
 {id=WavePlayer_9, properties {soundWaveAssetPtr=&BeastGirlStorm_Walk_04}}
]
properties {
 attenuationSettings=/BeastGirlStormClawsGameplay/Assets/Audio/ATT_BeastGirlStorm_Footsteps
 maxDistance=3220.3438
 soundClassObject=&Fort_Player_Foley
 volumeMultiplier=1.5
 attenuationOverrides {falloffDistance=10000, attenuationShapeExtents {x=10000}}
 concurrencyOverrides {maxCount=1, resolutionRule=PreventNew}
 concurrencySet [/BeastGirlStormClawsGameplay/Assets/Audio/CONC_BeastGirlStorm_FS]
}
