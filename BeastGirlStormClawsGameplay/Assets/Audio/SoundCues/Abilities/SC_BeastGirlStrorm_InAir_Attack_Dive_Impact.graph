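# Sound cue for the impact of the Beast Girl Storm Claws in-air dive attack, limited to one instance per owner.
# It mixes randomized close, medium and distant layers, each with its own attenuation and sound class.
# The close layer pairs two variant sets, one slightly delayed, while the distant layer reuses Storm King hammer sounds.
# Branches on islocal and isenemy play the mix loudest for the local player, lower for enemies, quietest for other non-enemies.
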SoundCue /BeastGirlStormClawsGameplay/Assets/Audio/SoundCues/Abilities/SC_BeastGirlStrorm_InAir_Attack_Dive_Impact
root=Branch_1
nodes [
 {
  id=Attenuation_1
  children [Random_3]
  properties {
   attenuationSettings=/BeastGirlStormClawsGameplay/Assets/Audio/ATT_BeastGirlStormClaws_Med
   attenuationOverrides {falloffDistance=2000, stereoSpread=2000, attenuationShapeExtents {x=800}}
  }
 }
 {
  id=Attenuation_2
  children [Modulator_0]
  properties {
   attenuationSettings=/BeastGirlStormClawsGameplay/Assets/Audio/ATT_BeastGirlStormClaws_Close
   attenuationOverrides {falloffDistance=2000, stereoSpread=2000, attenuationShapeExtents {x=800}}
  }
 }
 {
  id=Attenuation_3
  children [Random_2]
  properties {
   attenuationSettings=/BeastGirlStormClawsGameplay/Assets/Audio/ATT_BeastGirlStormClaws_Dist
   attenuationOverrides {bAttenuateWithLPF=true, falloffDistance=2000, lPFRadiusMax=5000, lPFRadiusMin=2000, attenuationShapeExtents {x=3000}}
  }
 }
 {id=Branch_1, children [Modulator_2,Branch_2,Modulator_2], properties {boolParameterName=islocal}}
 {id=Branch_2, children [Modulator_4,Modulator_3,Modulator_4], properties {boolParameterName=isenemy}}
 {id=Delay_0, children [Random_1], properties {delayMax=0.05, delayMin=0.05}}
 {id=Mixer_0, children [Random_0,Delay_0], properties {inputVolume [1,1]}}
 {id=Mixer_2, children [SoundClass_3,SoundClass_0,SoundClass_2], properties {inputVolume [1.2,1,1]}}
 {id=Modulator_0, children [Mixer_0]}
 {id=Modulator_2, children [Mixer_2], properties {pitchMax=1, pitchMin=1, volumeMax=1, volumeMin=1}}
 {id=Modulator_3, children [Mixer_2], properties {pitchMax=1, pitchMin=1, volumeMax=0.25, volumeMin=0.25}}
 {id=Modulator_4, children [Mixer_2], properties {pitchMax=1, pitchMin=1, volumeMax=0.75, volumeMin=0.75}}
 {id=Random_0, children [WavePlayer_16, WavePlayer_17, WavePlayer_18], properties {weights [1,1,1]}}
 {id=Random_1, children [WavePlayer_19, WavePlayer_20, WavePlayer_21, WavePlayer_22], properties {weights [1,1,1,1]}}
 {id=Random_2, children [WavePlayer_3, WavePlayer_4, WavePlayer_5, WavePlayer_6, WavePlayer_10], properties {weights [1,1,1,1,1]}}
 {id=Random_3, children [WavePlayer_12, WavePlayer_13, WavePlayer_14], properties {weights [1,1,1]}}
 {
  id=SoundClass_0
  children [Attenuation_1]
  properties {soundClassOverride=&Fort_WeaponRanged_Fire_3P_PSM}
 }
 {
  id=SoundClass_2
  children [Attenuation_3]
  properties {soundClassOverride=&Fort_WeaponRanged_Fire_3P_PSM}
 }
 {
  id=SoundClass_3
  children [Attenuation_2]
  properties {soundClassOverride=/Game/Sounds/FortSoundClassesAndModes/Fort_06_Player_Attacks_Abilities}
 }
 {
  id=WavePlayer_10
  properties {soundWaveAssetPtr=/Game/Sounds/Fort_Weapons/Storm_King/StormKing_Hammer/StormKing_Hammer_Heavy_Attack_Distant_05}
 }
 {
  id=WavePlayer_12
  properties {soundWaveAssetPtr=/BeastGirlStormClawsGameplay/Assets/Audio/Wavs/Abilities/BeastGirlStormClaws_InAir_Attack_Dive_Impact_Medium_01}
 }
 {
  id=WavePlayer_13
  properties {soundWaveAssetPtr=/BeastGirlStormClawsGameplay/Assets/Audio/Wavs/Abilities/BeastGirlStormClaws_InAir_Attack_Dive_Impact_Medium_02}
 }
 {
  id=WavePlayer_14
  properties {soundWaveAssetPtr=/BeastGirlStormClawsGameplay/Assets/Audio/Wavs/Abilities/BeastGirlStormClaws_InAir_Attack_Dive_Impact_Medium_03}
 }
 {
  id=WavePlayer_16
  properties {soundWaveAssetPtr=/BeastGirlStormClawsGameplay/Assets/Audio/Wavs/Abilities/BeastGirlStormClaws_InAir_Attack_Dive_Impact_Close_A_01}
 }
 {
  id=WavePlayer_17
  properties {soundWaveAssetPtr=/BeastGirlStormClawsGameplay/Assets/Audio/Wavs/Abilities/BeastGirlStormClaws_InAir_Attack_Dive_Impact_Close_A_02}
 }
 {
  id=WavePlayer_18
  properties {soundWaveAssetPtr=/BeastGirlStormClawsGameplay/Assets/Audio/Wavs/Abilities/BeastGirlStormClaws_InAir_Attack_Dive_Impact_Close_A_03}
 }
 {
  id=WavePlayer_19
  properties {soundWaveAssetPtr=/BeastGirlStormClawsGameplay/Assets/Audio/Wavs/Abilities/BeastGirlStormClaws_InAir_Attack_Dive_Impact_Close_B_01}
 }
 {
  id=WavePlayer_20
  properties {soundWaveAssetPtr=/BeastGirlStormClawsGameplay/Assets/Audio/Wavs/Abilities/BeastGirlStormClaws_InAir_Attack_Dive_Impact_Close_B_02}
 }
 {
  id=WavePlayer_21
  properties {soundWaveAssetPtr=/BeastGirlStormClawsGameplay/Assets/Audio/Wavs/Abilities/BeastGirlStormClaws_InAir_Attack_Dive_Impact_Close_B_03}
 }
 {
  id=WavePlayer_22
  properties {soundWaveAssetPtr=/BeastGirlStormClawsGameplay/Assets/Audio/Wavs/Abilities/BeastGirlStormClaws_InAir_Attack_Dive_Impact_Close_B_04}
 }
 {
  id=WavePlayer_3
  properties {soundWaveAssetPtr=/Game/Sounds/Fort_Weapons/Storm_King/StormKing_Hammer/StormKing_Hammer_Heavy_Attack_Distant_01}
 }
 {
  id=WavePlayer_4
  properties {soundWaveAssetPtr=/Game/Sounds/Fort_Weapons/Storm_King/StormKing_Hammer/StormKing_Hammer_Heavy_Attack_Distant_02}
 }
 {
  id=WavePlayer_5
  properties {soundWaveAssetPtr=/Game/Sounds/Fort_Weapons/Storm_King/StormKing_Hammer/StormKing_Hammer_Heavy_Attack_Distant_03}
 }
 {
  id=WavePlayer_6
  properties {soundWaveAssetPtr=/Game/Sounds/Fort_Weapons/Storm_King/StormKing_Hammer/StormKing_Hammer_Heavy_Attack_Distant_04}
 }
]
properties {
 attenuationSettings=/BeastGirlStormClawsGameplay/Assets/Audio/ATT_BeastGirlStormClaws_Lightning
 bOverrideConcurrency=true
 maxDistance=30100
 soundClassObject=&Fort_06_Player_Attacks_Abilities
 volumeMultiplier=0.5
 attenuationOverrides {distanceAlgorithm=NaturalSound, falloffDistance=3000, attenuationShapeExtents {x=1000}}
 concurrencyOverrides {bLimitToOwner=true, maxCount=1}
}
